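; Crossbow weapon Blueprint (GuineaPig unvault) that handles charge effects, fire audio and first-person scoping.
; While charging, a cosmetic-only looping timer feeds normalized charge into the ChargeUp particle parameter until charging ends.
; Firing picks charged or partial Sound Library events by charge state, in 1P or 3P variants by local view, and shows a sound indicator.
; Targeting enables scope targeting only when the cached player pawn has the first-person camera tag.
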(event EventOnStartCharge
  (bind _ns_guineapig_chargeup (Variables|Default|GetNS_GuineaPig_ChargeUp))
  ;[0.8439 0.3535 1] Update charge FX
  (Networking|CanExecuteCosmeticEvents
    (:True
      (Components|Activation|Activate _ns_guineapig_chargeup false)
      (Effects|Components|ParticleSystem|SetFloatParameter _ns_guineapig_chargeup "ChargeUp" 0.0)
      (bind _returnvalue (Utilities|Time|SetTimerbyEvent (delegate Custom|Set%20Charge%20Percentage) 0.1 true true))
      (Variables|Default|SetChargeUpdateHandle _returnvalue))))

(event Custom|Set%20Charge%20Percentage
  ;[0.8439 0.3535 1] Update charge FX
  (Effects|Components|ParticleSystem|SetFloatParameter (Variables|Default|GetNS_GuineaPig_ChargeUp) "ChargeUp" (/ (FortWeapon|GetChargePercent) (FortWeapon|GetCurrentMaxChargePercent))))

(event EventOnEndCharge
  ;[0.8439 0.3535 1] Update charge FX
  (Networking|CanExecuteCosmeticEvents
    (:True
      (Components|Activation|Deactivate (Variables|Default|GetNS_GuineaPig_ChargeUp))
      (Utilities|Time|ClearandInvalidateTimerbyHandle (Variables|Default|GetChargeUpdateHandle)))))

(event FortWeapon|EventOnEquippedAndAttached
  (|CallParentFunction)
  (Variables|Default|SetWeapon%20Sound%20Library%20Component (SoundLibrary|GetSoundLibraryComponent "(TagName=\"SoundLibrary.Weapon\")")))

(event EventPlayWeaponFireCustomAudio
  (params (bPersistentFire :type Boolean) (bSecondaryFire :type Boolean))
  ;[0.1875 0.1029 0.06738] Fire Sounds
  (|CallParentFunction bPersistentFire bSecondaryFire)
  (bind _instigator (Variables|Actor|GetInstigator)
    (:then
      (if (Pawn|IsLocallyViewed _instigator)
        (CallFunction|PlayFire1PSound)
        (else
          (CallFunction|PlayFire3PSound)))
      (SoundIndicator|ShowSoundIndicator (Variables|Default|GetSoundIndicatorComponent%20Secondary) (Transformation|GetActorLocation) 0 1.0 0))
    (:else)))

(event Custom|Play%20Fire%201P%20Sound
  ;[0.1875 0.1029 0.06738] Fire Sounds
  (switch Utilities|FlowControl|Switch|SwitchonEFortWeaponChargeStateForFireFX (FortWeapon|GetWeaponChargeStateforFireFX)
    (:Partial
      _)
    (:Full
      (bind _weapon_sound_library_component (Variables|Default|GetWeapon%20Sound%20Library%20Component :!label _label)
        (:then
          (SoundLibrary|PlayEvent _weapon_sound_library_component "(TagName=\"SoundLibrary.Weapon.Fire.Charged.1P\")"))
        (:else)))
    (:Over
      (goto _label))
    (:Max_None
      _))
  (bind _weapon_sound_library_component_2 (Variables|Default|GetWeapon%20Sound%20Library%20Component)
    (:then
      (SoundLibrary|PlayEvent _weapon_sound_library_component_2 "(TagName=\"SoundLibrary.Weapon.Fire.Partial.1P\")"))
    (:else)))

(event Custom|Play%20Fire%203P%20Sound
  ;[0.1875 0.1029 0.06738] Fire Sounds
  (switch Utilities|FlowControl|Switch|SwitchonEFortWeaponChargeStateForFireFX (FortWeapon|GetWeaponChargeStateforFireFX)
    (:Partial
      _)
    (:Full
      (bind _weapon_sound_library_component (Variables|Default|GetWeapon%20Sound%20Library%20Component :!label _label)
        (:then
          (SoundLibrary|PlayEvent _weapon_sound_library_component "(TagName=\"SoundLibrary.Weapon.Fire.Charged.3P\")"))
        (:else)))
    (:Over
      (goto _label))
    (:Max_None
      _))
  (bind _weapon_sound_library_component_2 (Variables|Default|GetWeapon%20Sound%20Library%20Component)
    (:then
      (SoundLibrary|PlayEvent _weapon_sound_library_component_2 "(TagName=\"SoundLibrary.Weapon.Fire.Partial.3P\")"))
    (:else)))

(event EventOnSetTargeting
  (params (bNewIsTargeting :type Boolean))
  ; First Person Scoping
  (if (GameplayTags|HasMatchingGameplayTag (Variables|Default|GetAs%20Fort%20Player%20Pawn) "(TagName=\"Gameplay.Camera.FirstPerson\")")
    (Scope|SetScopeTargeting true)
    (else
      (Scope|SetScopeTargeting false)))
  (|CallParentFunction bNewIsTargeting))

(event EventBeginPlay
  (bind _returnvalue (Game|GetInstigator))
  ;[1 0.6311 0.9916 0.5] Event Begin Play Cast to FortPlayerPawn
  (|CallParentFunction)
  (Utilities|IsValid _returnvalue
    (:Is%20Valid
      (bind _asfort_player_pawn (Utilities|Casting|CastToFortPlayerPawn _returnvalue))
      (Variables|Default|SetAs%20Fort%20Player%20Pawn _asfort_player_pawn))))

(fn ConstructionScript
  (|CallParentFunction))
